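# Physics asset for the F_MED_Dark_Ninja_Purple backpack, defining collision bodies and joints for its left and right scythes and ropes.
# A central capsule on the pack bone is the hub, with both scythe capsules attached under tight swing and twist limits.
# Each rope is a four-box chain with wide swing and tight twist limits, joined to the pack through a small offset capsule whose angular motion is free.
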PhysicsAsset /Game/Accessories/FORT_Backpacks/Backpack_F_MED_Dark_Ninja_Purple/Meshes/F_MED_Dark_Ninja_Purple_Pack_Physics
bodies {
 dyn_rope_l_01 [{shapeType=Box, extentX=1.9716, extentY=5.389, extentZ=1.5486, center=<1.0595 0 0>}]
 dyn_rope_l_02 [{shapeType=Box, extentX=1.9716, extentY=5.389, extentZ=1.5486, center=<1.0595 0 0>}]
 dyn_rope_l_03 [{shapeType=Box, extentX=1.9716, extentY=5.389, extentZ=1.5486, center=<1.0595 0 0>}]
 dyn_rope_l_04 [{shapeType=Box, extentX=1.9716, extentY=5.389, extentZ=1.5486, center=<1.0595 0 0>}]
 dyn_rope_l_Offset [{shapeType=Capsule, length=0.5, radius=0.5}]
 dyn_rope_r_01 [{shapeType=Box, extentX=1.9716, extentY=4.9654, extentZ=1.5486, center=<1.0595 0 0>}]
 dyn_rope_r_02 [{shapeType=Box, extentX=1.9716, extentY=5.032, extentZ=1.5486, center=<1.0595 0 0>}]
 dyn_rope_r_03 [{shapeType=Box, extentX=1.9716, extentY=5.032, extentZ=1.5486, center=<1.0595 0 0>}]
 dyn_rope_r_04 [{shapeType=Box, extentX=1.9716, extentY=5.032, extentZ=1.5486, center=<1.0595 0 0>}]
 dyn_rope_r_Offset [{shapeType=Capsule, length=0.5, radius=0.5}]
 dyn_scythe_l [{shapeType=Capsule, length=2, radius=2, center=<-2.8076 0.4549 -0.0121>, rotation=<rot 87.3367 23.8687 34.2431>}]
 dyn_scythe_r [{shapeType=Capsule, length=2, radius=2, center=<2.848 -0.4 0.0002>, rotation=<rot 90 90 99.0692>}]
 pack [{shapeType=Capsule, length=18.2435, radius=6.1629, center=<0.0044 2.7623 -5.8376>, rotation=<rot 0 -90 90>}]
}
constraints [
 {bone1Name=dyn_scythe_r, bone2Name=pack, swing1LimitDegrees=5, swing2LimitDegrees=5, twistLimitDegrees=5}
 {bone1Name=dyn_scythe_l, bone2Name=pack, swing1LimitDegrees=5, swing2LimitDegrees=5, twistLimitDegrees=5}
 {
  bone1Name=dyn_rope_l_Offset
  bone2Name=pack
  swing1LimitDegrees=45
  swing1Motion=Free
  swing2LimitDegrees=45
  swing2Motion=Free
  twistLimitDegrees=45
  twistMotion=Free
 }
 {
  bone1Name=dyn_rope_r_Offset
  bone2Name=pack
  swing1LimitDegrees=45
  swing1Motion=Free
  swing2LimitDegrees=45
  swing2Motion=Free
  twistLimitDegrees=45
  twistMotion=Free
 }
 {bone1Name=dyn_rope_r_01, bone2Name=dyn_rope_r_Offset, swing1LimitDegrees=45, swing2LimitDegrees=45, twistLimitDegrees=5}
 {bone1Name=dyn_rope_r_02, bone2Name=dyn_rope_r_01, swing1LimitDegrees=45, swing2LimitDegrees=45, twistLimitDegrees=5}
 {bone1Name=dyn_rope_r_03, bone2Name=dyn_rope_r_02, swing1LimitDegrees=45, swing2LimitDegrees=45, twistLimitDegrees=5}
 {bone1Name=dyn_rope_r_04, bone2Name=dyn_rope_r_03, swing1LimitDegrees=45, swing2LimitDegrees=45, twistLimitDegrees=5}
 {bone1Name=dyn_rope_l_01, bone2Name=dyn_rope_l_Offset, swing1LimitDegrees=45, swing2LimitDegrees=45, twistLimitDegrees=5}
 {bone1Name=dyn_rope_l_02, bone2Name=dyn_rope_l_01, swing1LimitDegrees=45, swing2LimitDegrees=45, twistLimitDegrees=5}
 {bone1Name=dyn_rope_l_03, bone2Name=dyn_rope_l_02, swing1LimitDegrees=45, swing2LimitDegrees=45, twistLimitDegrees=5}
 {bone1Name=dyn_rope_l_04, bone2Name=dyn_rope_l_03, swing1LimitDegrees=45, swing2LimitDegrees=45, twistLimitDegrees=5}
]
